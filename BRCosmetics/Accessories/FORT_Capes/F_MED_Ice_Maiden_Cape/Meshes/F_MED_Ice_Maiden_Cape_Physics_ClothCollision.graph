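# Physics asset for the Ice Maiden cape backpack accessory, named as a cloth collision asset.
# It holds a single capsule body on the backpack_root bone, offset and rotated from that bone.
# It has no constraints, so the capsule serves as a simple collision volume, likely for the cape's cloth simulation.
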PhysicsAsset /BRCosmetics/Accessories/FORT_Capes/F_MED_Ice_Maiden_Cape/Meshes/F_MED_Ice_Maiden_Cape_Physics_ClothCollision
bodies {backpack_root [{shapeType=Capsule, length=10.7961, radius=6.341, center=<0.7877 -7.6659 -7.6463>, rotation=<rot 0 93.289 -93.0085>}]}
constraints []
